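; Gameplay ability for the sprint charge of the rideable creature Burt, toggling collision reactions on its avatar actor while active.
; On activation it commits, then calls Run on the avatar's FortAthenaCollisionReactionComponent without checking the commit result.
; On end it calls Stop on that component, whether cancelled or not.
; Both paths validate actor and component; the ability never ends itself, and its passive activation failure event is empty.
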(event Ability|EventActivateAbility
  (bind _returnvalue (Ability|GetAvatarActorfromActorInfo))
  (bind _returnvalue_2 (Actor|GetComponentbyClass _returnvalue "/Script/FortniteGame.FortAthenaCollisionReactionComponent"))
  (Ability|CommitAbility)
  (Utilities|IsValid _returnvalue
    (:Is%20Valid
      (Utilities|IsValid _returnvalue_2
        (:Is%20Valid
          (CollisionReactions|Run _returnvalue_2))))))

(event Ability|EventOnEndAbility
  (params (bWasCancelled :type Boolean))
  (bind _returnvalue (Ability|GetAvatarActorfromActorInfo))
  (bind _returnvalue_2 (Actor|GetComponentbyClass _returnvalue "/Script/FortniteGame.FortAthenaCollisionReactionComponent"))
  (Utilities|IsValid _returnvalue
    (:Is%20Valid
      (Utilities|IsValid _returnvalue_2
        (:Is%20Valid
          (CollisionReactions|Stop _returnvalue_2))))))

(event Ability|EventFailedToActivatePassiveAbility
  (params (ActorInfo :type Gameplay%20Ability%20Actor%20Info%20Structure)))
