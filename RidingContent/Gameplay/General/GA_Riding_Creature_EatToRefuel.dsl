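; Event-activated gameplay ability that refuels a rideable creature when it eats.
; It scales the event magnitude by an energy modifier curve from the RidingData table and passes the negated, scaled result to OnEattoRefuel on the avatar's creature ridable component.
; If the instigator tags match the Slap Berry tags, it also applies the Slap Berry gameplay effect to the activating wildlife pawn.
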(event Ability|EventActivateAbilityFromEvent
  (params (EventData_EventTag :type Gameplay%20Tag%20Structure) (EventData_Instigator :type Actor%20Object%20Reference) (EventData_Target :type Actor%20Object%20Reference) (EventData_OptionalObject :type Object%20Reference) (EventData_OptionalObject2 :type Object%20Reference) (EventData_ContextHandle :type Gameplay%20Effect%20Context%20Handle%20Structure) (EventData_InstigatorTags :type Gameplay%20Tag%20Container%20Structure) (EventData_TargetTags :type Gameplay%20Tag%20Container%20Structure) (EventData_EventMagnitude :type Float%20%28single-precision%29) (EventData_TargetData :type Gameplay%20Ability%20Target%20Data%20Handle%20Structure) (EventData_InstancedEventData :type Instanced%20Struct%20Structure))
  (bind _returnvalue (Ability|GetAvatarActorfromActorInfo))
  (bind _returnvalue_2 (Actor|GetComponentbyClass _returnvalue "/Script/RidingCodeRuntime.ControllableRidableComponent"))
  (Ability|CommitAbility)
  (Utilities|IsValid _returnvalue_2
    (:Is%20Valid
      (bind _ascreature_base_ridable_component (Utilities|Casting|CastToCreatureBaseRidableComponent _returnvalue_2)
        (:then
          (bind _outxy (DataTable|EvaluateCurveTableRow "/RidingContent/DataTables/RidingData.RidingData" "Riding.Creature.Stamina.EatItem.EnergyModifier")
            (:RowFound
              (Class|CreatureBaseRidableComponent|OnEattoRefuel _ascreature_base_ridable_component (* (* EventData_EventMagnitude _outxy) -0.01))
              (if (GameplayTags|HasAllTags EventData_InstigatorTags (Variables|Default|GetSlap%20Berry%20Tags) true)
                (bind _asnpc_pawn_wildlife_parent (Utilities|Casting|CastToNPC_Pawn_Wildlife_Parent (Ability|GetActivatingPawn))
                  (:then
                    (GameplayEffects|ApplyGameplayEffectToSelf :self (Variables|/Script/FortniteGame.FortAIPawn|GetAIPawnAbilitySystemComponent _asnpc_pawn_wildlife_parent) :GameplayEffectClass (Variables|Default|GetSlap%20Berry%20Gameplay%20Effect) :Level 0.0)
                    (Ability|EndAbility))
                  (:CastFailed
                    (Ability|EndAbility)))
                (else
                  (Ability|EndAbility))))
            (:RowNotFound
              (Ability|EndAbility))))
        (:CastFailed
          (Ability|EndAbility))))
    (:Is%20Not%20Valid
      (Ability|EndAbility))))
