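; Gameplay ability that makes a rider's active mount sprint.
; On activation it holsters the rider's weapon, cancels the build bar, applies the creature's sprint effect to the mount and sets bIsSprinting.
; It waits on each cancel-sprint tag, removing the player's GE_Riding_Player_Sprint effect when one is added.
; Ending the ability reverses the holster, removes the mount's effect and clears the flag, forcing a net update on the mount.
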(event Ability|EventActivateAbility
  (bind _returnvalue (Ability|GetAvatarActorfromActorInfo))
  (bind _returnvalue_2 (Actor|GetComponentbyClass _returnvalue "/Script/RidingCodeRuntime.ControllingRiderComponent"))
  (bind _donotuse_activeridable (Variables|/Script/RidingCodeRuntime.RiderComponent|GetDoNotUse_ActiveRidable _returnvalue_2))
  (bind _returnvalue_3 (Components|GetOwner _donotuse_activeridable))
  (bind _returnvalue_4 (Ability|GetAbilitySystemComponent _returnvalue))
  (Ability|CommitAbility)
  (Utilities|IsValid _returnvalue
    (:Is%20Valid
      (Utilities|IsValid _returnvalue_2
        (:Is%20Valid
          (Utilities|IsValid _donotuse_activeridable
            (:Is%20Valid
              (Utilities|IsValid _returnvalue_3
                (:Is%20Valid
                  (bind _asfort_pawn (Utilities|Casting|CastToFortPawn _returnvalue))
                  (Utilities|IsValid _asfort_pawn
                    (:Is%20Valid
                      (Weapon|SetHolsterWeapon _asfort_pawn 0 true false "Riding")
                      (bind _ascreature_base_ridable_component (Utilities|Casting|CastToCreatureBaseRidableComponent _donotuse_activeridable))
                      (Class|ControllableRidableComponent|CancelBuildBar _ascreature_base_ridable_component)
                      (Creative|GameplayEffects|SimpleApplyGameplayEffectToActor (Variables|/RidingContent/Components/CreatureBaseRidableComponent.CreatureBaseRidableComponent_C|GetGE_Creature_Sprint _ascreature_base_ridable_component) _returnvalue_3)
                      (Networking|ForceNetUpdate _returnvalue_3)
                      (Variables|/Script/RidingCodeRuntime.RidableComponent|SetbIsSprinting true _donotuse_activeridable)
                      (for _array_element (Variables|Default|GetTags%20to%20cancel%20sprinting)
                        (Ability|Tasks|WaitGameplayTagAdd _array_element
                          (:Added
                            (Utilities|IsValid _returnvalue_4
                              (:Is%20Valid
                                (GameplayEffects|RemoveActiveGameplayEffectbySourceEffect _returnvalue_4 "/RidingContent/Gameplay/General/GE_Riding_Player_Sprint.GE_Riding_Player_Sprint_C" 0 -1)))))))))))))))))

(event Ability|EventFailedToActivatePassiveAbility
  (params (ActorInfo :type Gameplay%20Ability%20Actor%20Info%20Structure)))

(event Ability|EventOnEndAbility
  (params (bWasCancelled :type Boolean))
  (bind _returnvalue (Ability|GetAvatarActorfromActorInfo))
  (bind _returnvalue_2 (Actor|GetComponentbyClass _returnvalue "/Script/RidingCodeRuntime.ControllingRiderComponent"))
  (bind _donotuse_activeridable (Variables|/Script/RidingCodeRuntime.RiderComponent|GetDoNotUse_ActiveRidable _returnvalue_2))
  (bind _returnvalue_3 (Components|GetOwner _donotuse_activeridable))
  (bind _returnvalue_4 (Ability|GetAbilitySystemComponent _returnvalue_3))
  (Utilities|IsValid _returnvalue
    (:Is%20Valid
      (bind _asfort_pawn (Utilities|Casting|CastToFortPawn _returnvalue))
      (Utilities|IsValid _asfort_pawn
        (:Is%20Valid
          (Utilities|IsValid _returnvalue_2
            (:Is%20Valid
              (Utilities|IsValid _donotuse_activeridable
                (:Is%20Valid
                  (Utilities|IsValid _returnvalue_3
                    (:Is%20Valid
                      (Weapon|SetHolsterWeapon _asfort_pawn 0 false true "Riding")
                      (Utilities|IsValid _returnvalue_4
                        (:Is%20Valid
                          (bind _ascreature_base_ridable_component (Utilities|Casting|CastToCreatureBaseRidableComponent _donotuse_activeridable))
                          (GameplayEffects|RemoveActiveGameplayEffectbySourceEffect _returnvalue_4 (Variables|/RidingContent/Components/CreatureBaseRidableComponent.CreatureBaseRidableComponent_C|GetGE_Creature_Sprint _ascreature_base_ridable_component) 0 -1)
                          (Networking|ForceNetUpdate _returnvalue_3)
                          (Variables|/Script/RidingCodeRuntime.RidableComponent|SetbIsSprinting false _donotuse_activeridable)
                          (Utilities|Casting|CastToCreatureBaseRidableComponent _donotuse_activeridable))))))))))))))

(fn OnRep_CancelSprintDispatcher)
